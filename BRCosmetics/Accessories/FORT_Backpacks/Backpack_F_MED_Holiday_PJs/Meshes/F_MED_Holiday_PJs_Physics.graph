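# Physics asset for the Holiday PJs backpack cosmetic, defining collision and joint limits for its dangling accessory bones.
# Each body is a single box: the pack, a candy cane, a crowbar, a bow and two bow ribbons.
# Cane, crowbar and bow are constrained to the pack, and both ribbons chain off the bow.
# Limits loosen from the nearly rigid cane and crowbar through the bow to the widely swinging ribbons.
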PhysicsAsset /BRCosmetics/Accessories/FORT_Backpacks/Backpack_F_MED_Holiday_PJs/Meshes/F_MED_Holiday_PJs_Physics
bodies {
 dyn_bow [{shapeType=Box, extentX=8.3614, extentY=17.7299, extentZ=3.3153, center=<1.3773 0 0>, rotation=<rot 90 0 0>}]
 dyn_bow_ribbon_a [{shapeType=Box, extentX=1.435, extentY=4, extentZ=10, center=<5 0 0>, rotation=<rot 90 0 0>}]
 dyn_bow_ribbon_b [{shapeType=Box, extentX=1.435, extentY=4, extentZ=6.5773, center=<3.2686 0 0>, rotation=<rot 90 0 0>}]
 dyn_cane [{shapeType=Box, extentX=10.4736, extentY=3.5007, extentZ=41.0448, center=<1.6705 -0.7396 -2.3408>, rotation=<rot 90 0 0>}]
 dyn_crowbar [{shapeType=Box, extentX=8.7177, extentY=3.5099, extentZ=42.0982, center=<1.597 0.1897 -0.0288>, rotation=<rot 90 0 0>}]
 pack [{shapeType=Box, extentX=3.0744, extentY=4.0681, extentZ=8.3799, center=<-0.1124 -4.6124 -0.5953>, rotation=<rot 31.1227 -90 -11.4346>}]
}
constraints [
 {bone1Name=dyn_cane, bone2Name=pack, swing1LimitDegrees=10, swing2LimitDegrees=10, twistLimitDegrees=1}
 {bone1Name=dyn_crowbar, bone2Name=pack, swing1LimitDegrees=10, swing2LimitDegrees=10, twistLimitDegrees=1}
 {bone1Name=dyn_bow, bone2Name=pack, swing1LimitDegrees=45, swing2LimitDegrees=45, twistLimitDegrees=15}
 {bone1Name=dyn_bow_ribbon_a, bone2Name=dyn_bow, swing1LimitDegrees=75, swing2LimitDegrees=75, twistLimitDegrees=10}
 {bone1Name=dyn_bow_ribbon_b, bone2Name=dyn_bow, swing1LimitDegrees=75, swing2LimitDegrees=75, twistLimitDegrees=10}
]
